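# Translucent lit material for Niagara mesh particles in the SuperSpike emote FX, producing a particle-tinted Fresnel effect.
# It adds an inverted Fresnel to a plain Fresnel, each with its own exponent and reflect-fraction parameters.
# The inverted Fresnel's normal is a float2 built from the gadget normal map's red channel twice.
# The sum is multiplied separately by particle color RGB and alpha.
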Material /BRCosmetics/Animation/Game/MainPlayer/Emotes/SuperSpike/FX/M_SuperSpike_Fresnel
graph {
 properties {bAutomaticallySetUsageInEditor=false, bUsedWithNiagaraMeshParticles=true, blendMode=BLEND_Translucent, usedShadingModels=Default Lit}
 nodes [
  n1 = ParticleColor()
  n2 = ScalarParameter() {defaultValue=2, parameterName=FresnelStrength}
  n3 = ScalarParameter() {parameterName=FresnelReflectionStrength}
  n4 = TextureSample() {
   samplerSource=SSM_Wrap_WorldGroupSettings
   samplerType=SAMPLERTYPE_Normal
   texture=/BRCosmetics/Animation/Props/Gadgets_SuperSpike/Textures/T_SuperSpike_Gadgets_N
  }
  n5 = MaterialFunctionCall(X=@n4.R, Y=@n4.R) {materialFunction=&MakeFloat2}
  n6 = Add(
   A=OneMinus(None=Fresnel(ExponentIn=@n2, BaseReflectFractionIn=@n3, Normal=@n5.Result) {exponent=2})
   B=Fresnel(
    ExponentIn=ScalarParameter() {defaultValue=2, parameterName=AddFresnelStrength}
    BaseReflectFractionIn=ScalarParameter() {defaultValue=0.04, parameterName=AddFresnelReflection}
   ) {exponent=2}
  )
  Multiply(A=@n1.RGB, B=@n6)
  Multiply(A=@n6, B=@n1.A)
 ]
}
