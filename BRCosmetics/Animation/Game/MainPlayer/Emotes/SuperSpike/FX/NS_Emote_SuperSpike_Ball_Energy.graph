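# Niagara effect for the SuperSpike emote that puts a glowing energy look on the volleyball prop, with every emitter following bone JNT_Ball_A of SuperSpike_Gadget.
# Energy continuously spawns short-lived meshes and sprites, with per-particle variables for core and glow layers.
# Flash, Shockwave and Volleyball_Outlines add burst-spawned layered flare sprites, an expanding mesh ring and an outline mesh.
# FrontEnd modifiers retune color, alpha and scale outside gameplay, and time of day dims the shockwave.
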NiagaraSystem /BRCosmetics/Animation/Game/MainPlayer/Emotes/SuperSpike/FX/NS_Emote_SuperSpike_Ball_Energy
dependencies {
 usedDataInterfaces [&NiagaraDataInterfaceCurve, &NiagaraDataInterfaceVectorCurve, &NiagaraDataInterfaceCamera]
 usedDynamicInputs []
 usedModules [&SystemState]
}
emitters {
 Energy {
  emitterUpdateScript [
   &EmitterState(Inactive Response=::NewEnumerator1, Life Cycle Mode=::NewEnumerator1, Loop Behavior=::NewEnumerator1, UseLoopDelay=true)
   &SpawnRate(SpawnRate=10, Use Spawn Probability=false)
  ]
  particleSpawnScript [
   &InitializeParticle(
    Color=<rgba 3.8098 0.9532 0.5358 1>
    Lifetime=0.2
    Mesh Scale=&Multiply_Vector(
     A=&RandomRangeVector(Minimum=<1 1 1>, Maximum=<1 1.5 1.5>)
     B=&VectorFromFloat(Value=2.5)
    )
    Mesh Scale Mode=::NewEnumerator4
    Position Offset Coordinate Space=::Local
    UsePositionOffset=true
    Write Material Random=true
   )
   &InitialMeshOrientation(Mesh Orientation Mode=::NewEnumerator2, Rotation=<0 0.25 0>, Use Rotation Vector=true)
   /SetVariables(Particles.Particle_BG_Color:LinearColor=<rgba 0.3698 0.0628 0 0.3>) {isSetParameters=true}
  ]
  particleUpdateScript [
   &ScaleColor(
    Scale Alpha=&FN_FrontEndFloatModifier(Gameplay=1, Front End=0.55)
    Scale RGB=&FN_FrontEndVector3Modifier(Gameplay=<1 1 1>, Front End=<0.8 1.2 1>)
    ScaleA=true
    ScaleRGB=true
   )
   &ParticleState()
   &SpriteRotationRate(Rotation Rate=10)
   &DynamicMaterialParameters(
    Index 0 Param 1=1.2
    Index 0 Param 2=2
    Index 0 Param 3=1
    Index 0 Param 4=1
    Module.Index0Float0Write=true
    Module.Index0Float1Write=true
    Module.Index0Float2Write=true
    Module.Index0Float3Write=true
   )
   &ScaleMeshSize(Scale Factor=&Multiply_VectorByFloat(
     Vector=<0.6 0.5 0.5>
     Float=&FloatFromCurve(FloatCurve=DataInterface() {propertyValues {Curve {keys [{interpMode=RCIM_Cubic, tangentMode=RCTM_User, time=0, value=0.4}, {interpMode=RCIM_Cubic, tangentMode=RCTM_User, time=1, value=0.575}]}}})
    ))
   &SkeletalMeshLocation(
    Bone Sampling Mode=::NewEnumerator4
    MaskSpawnGroup=false
    Orientation Sampling=::NewEnumerator1
    Skeletal Mesh=DataInterface() {propertyValues {
      DefaultMesh=/BRCosmetics/Animation/Props/Gadgets_SuperSpike/Meshes/SuperSpike_Gadget
      ExcludeBoneName=root
      OptionalComponentClass=null
      PreviewMesh=&SuperSpike_Gadget
      SkinningMode=SkinOnTheFly
      SoftSourceActor=null
      SourceMode=Default
      UvSetIndex=0
      WholeMeshLOD=-1
      bExcludeBone=false
      bReadDeformedGeometry=true
      bRequireCurrentFrameData=true
      ComponentTags []
      FilteredBones [JNT_Ball_A]
      FilteredSockets []
      MeshUserParameter {parameter {name=None, typeDefHandle {registeredTypeIndex=1}, varData []}}
      SamplingRegions []
     }}
   )
   /SetVariables(Particles.Position=&AddVectorToPosition(Position=$Particles.Position, Vector=<0 0 -21>)) {isSetParameters=true}
   &ScaleColor(Scale Alpha=0.8, ScaleA=true, ScaleRGB=true) {module=ScaleColor002}
   /SetVariables(
    Particles.AirGlowColor:LinearColor=<rgba 0.2865 0 0.1021 0.25>
    Particles.AirGlowPosition:NiagaraPosition=&AddVectorToPosition(Position=$Particles.Position, Vector=<0 0 30>)
    Particles.AirGlowSize:Vector2f=<150 150>
   ) {isSetParameters=true}
   /SetVariables(
    Particles.CoreCameraOffset=10
    Particles.CoreColor:LinearColor=<rgba 1 0.7416 0 1>
    Particles.CorePosition:NiagaraPosition=&AddVectorToPosition(Position=$Particles.Position, Vector=<0 0 20>)
    Particles.CoreRotation:NiagaraFloat=&Add_Float(A=30, B=&Multiply_Float(A=$Emitter.Age, B=150))
    Particles.CoreSize:Vector2f=<100 100>
   ) {isSetParameters=true}
   /SetVariables(
    Particles.BallGlowCameraOffset=20
    Particles.BallGlowColor:LinearColor=<rgba 0.9844 0.3761 0.1077 30>
    Particles.BallGlowPosiiton:NiagaraPosition=&AddVectorToPosition(Position=$Particles.Position, Vector=<0 0 20>)
    Particles.BallGlowSize:Vector2f=<30 30>
   ) {isSetParameters=true}
   /SetVariables(
    Particles.EnergyPositionOffset:NiagaraPosition=&AddVectorToPosition(Position=$Particles.Position, Vector=<0 0 20>)
    Particles.MeshCameraOffset=-125
    Particles.ScaleOffset:Vector3f=&Add_Vector(A=$Particles.Scale, B=&VectorFromFloat(Value=0.2))
   ) {isSetParameters=true}
  ]
  rendererClasses [&NiagaraMeshRendererProperties, &NiagaraSpriteRendererProperties]
  renderers [
   {rendererClass=&NiagaraMeshRendererProperties, rendererIndex=0}
   {rendererClass=&NiagaraMeshRendererProperties, rendererIndex=1}
   {rendererClass=&NiagaraSpriteRendererProperties, rendererIndex=2}
   {rendererClass=&NiagaraSpriteRendererProperties, rendererIndex=3}
  ]
 }
 Flash {
  emitterUpdateScript [
   &EmitterState(Life Cycle Mode=::NewEnumerator1, Loop Behavior=::NewEnumerator1, Loop Duration=1, UseLoopDelay=false)
   &SpawnBurst_Instantaneous(Spawn Time=0.01, Use Spawn Probability=false, UseLoopCountLimit=false)
   /SetVariables(Emitter.Cam Pos:NiagaraPosition=&ReturnCameraVector(Camera Property=::NewEnumerator0, Coordinate Space=::Simulation)) {isSetParameters=true}
  ]
  particleSpawnScript [&InitializeParticle(
    Color=<rgba 1 0.267 0.0344 1>
    Lifetime Max=0.23
    Lifetime Min=0.18
    Lifetime Mode=::NewEnumerator1
    Mass Max=2
    Mass Mode=::NewEnumerator3
    Position Offset Coordinate Space=::Local
    Sprite Rotation Mode=::NewEnumerator2
    Sprite Size Mode=::NewEnumerator3
    Uniform Sprite Size=&RandomRangeFloat(Minimum=100, Maximum=100)
    UsePositionOffset=true
    Write Material Random=true
   )]
  particleUpdateScript [
   &ScaleColor(
    Scale Alpha=&FN_FrontEndFloatModifier(Gameplay=1, Front End=1)
    Scale RGB=&FN_FrontEndVector3Modifier(Gameplay=<1 1 1>, Front End=<0.9 2 0.5>)
    ScaleA=true
    ScaleRGB=true
   ) {module=ScaleColor001}
   &ParticleState()
   &SkeletalMeshLocation(
    Bone Sampling Mode=::NewEnumerator4
    MaskSpawnGroup=false
    Orientation Sampling=::NewEnumerator1
    Skeletal Mesh=DataInterface() {propertyValues {
      DefaultMesh=&SuperSpike_Gadget
      ExcludeBoneName=root
      OptionalComponentClass=null
      PreviewMesh=&SuperSpike_Gadget
      SkinningMode=SkinOnTheFly
      SoftSourceActor=null
      SourceMode=Default
      UvSetIndex=0
      WholeMeshLOD=-1
      bExcludeBone=false
      bReadDeformedGeometry=true
      bRequireCurrentFrameData=true
      ComponentTags []
      FilteredBones [JNT_Ball_A]
      FilteredSockets []
      MeshUserParameter {parameter {name=None, typeDefHandle {registeredTypeIndex=1}, varData []}}
      SamplingRegions []
     }}
   )
   &CameraOffset(Camera Offset Amount=50) {module=CameraOffset001}
   &ScaleColor(
    Scale Alpha=&FloatFromCurve(
     FloatCurve=DataInterface() {propertyValues {Curve {keys [
         <at 0 0 RCIM_Cubic>
         {arriveTangent=0.1895, interpMode=RCIM_Cubic, leaveTangent=0.1895, tangentMode=RCTM_User, time=0.1, value=1}
         {arriveTangent=-0.0202, interpMode=RCIM_Cubic, leaveTangent=-0.0202, tangentMode=RCTM_User, time=0.7, value=1}
         <at 1.0011 0.002 RCIM_Cubic>
        ]}}}
     Scale Curve=10
    )
    ScaleA=true
    ScaleRGB=true
   )
   &ScaleSpriteSize(Scale Factor=&Vector2DFromFloat(Value=&FloatFromCurve(FloatCurve=DataInterface() {propertyValues {Curve {keys [
          {interpMode=RCIM_Cubic, tangentMode=RCTM_User, time=0, value=0}
          {arriveTangent=-0.0009, interpMode=RCIM_Cubic, leaveTangent=-0.0009, tangentMode=RCTM_User, time=0.2, value=1}
          {tangentMode=RCTM_User, time=1, value=0}
         ]}}})))
   /SetVariables(
    Particles.StarColor:LinearColor=&MultiplyLinearColor(
     Linear Color=&MultiplyLinearColor(
      Linear Color=$Particles.Color
      Scale Factor=&MakeVector4_FromLinearColor(LinearColor=<rgba 1 0.2028 0.1263 0.5>)
     )
     Scale Factor=&MakeVector4(
      X=&FN_FrontEndFloatModifier(Gameplay=1, Front End=1)
      Y=&FN_FrontEndFloatModifier(Gameplay=1, Front End=1.5)
      Z=&FN_FrontEndFloatModifier(Gameplay=1, Front End=0)
      W=&FN_FrontEndFloatModifier(Gameplay=1, Front End=1)
     )
    )
    Particles.StarRotation:NiagaraFloat=&RandomRangeFloat(Minimum=0, Maximum=360)
    Particles.StarSize:Vector2f=&Multiply_Vector2D(
     A=&Multiply_Vector2(A=$Particles.SpriteSize, B=<2.75 2.75>)
     B=&FN_FrontEndVector2Modifier(Front End:Vector2f=<0.8 0.8>, Gameplay:Vector2f=<1 1>)
    )
   ) {isSetParameters=true}
   /SetVariables(
    Particles.StreaksColor:LinearColor=&MultiplyLinearColor(
     Linear Color=$Particles.Color
     Scale Factor=&MakeVector4_FromLinearColor(LinearColor=<rgba 1 0.625 0 0.01>)
    )
    Particles.StreaksRotation:NiagaraFloat=&RandomRangeFloat(Minimum=0, Maximum=360)
    Particles.StreaksSize:Vector2f=&Multiply_Vector2(A=$Particles.SpriteSize, B=<7 7>)
   ) {isSetParameters=true}
   /SetVariables(
    Particles.AnamorphicColor:LinearColor=&MultiplyLinearColor(
     Linear Color=$Particles.Color
     Scale Factor=&MakeVector4_FromLinearColor(LinearColor=<rgba 1 0.8438 0.1968 0.01>)
    )
    Particles.AnamorphicRotation=90
    Particles.AnamorphicSize:Vector2f=&Multiply_Vector2(A=$Particles.SpriteSize, B=<3 9>)
   ) {isSetParameters=true}
   /SetVariables(
    Particles.HaloColor:LinearColor=&MultiplyLinearColor(
     Linear Color=$Particles.Color
     Scale Factor=&MakeVector4_FromLinearColor(LinearColor=<rgba 1 0.7936 0.2027 0.025>)
    )
    Particles.HaloSize:Vector2f=&Multiply_Vector2(A=$Particles.SpriteSize, B=<5 5>)
   ) {isSetParameters=true}
   /SetVariables(
    Particles.HaloColor_02:LinearColor=&MultiplyLinearColor(Linear Color=$Particles.Color, Scale Factor=&MakeVector4_FromLinearColor(LinearColor=<rgba 1 1 1 0.05>))
    Particles.HaloSize_02:Vector2f=&Multiply_Vector2(A=$Particles.SpriteSize, B=<1.5 1.5>)
   ) {isSetParameters=true}
   /SetVariables(
    Particles.AirGlowColor:LinearColor=&MultiplyLinearColor(
     Linear Color=&MultiplyLinearColor(
      Linear Color=$Particles.Color
      Scale Factor=&MakeVector4_FromLinearColor(LinearColor=<rgba 1 0.917 0.2383 0.01>)
     )
     Scale Factor=&MakeVector4(
      X=&FN_FrontEndFloatModifier(Gameplay=1, Front End=0.8)
      Y=&FN_FrontEndFloatModifier(Gameplay=1, Front End=1.5)
      Z=&FN_FrontEndFloatModifier(Gameplay=1, Front End=1)
      W=&FN_FrontEndFloatModifier(Gameplay=1, Front End=0)
     )
    )
    Particles.AirGlowSize:Vector2f=&Multiply_Vector2(A=$Particles.SpriteSize, B=<1.5 1.5>)
   ) {isSetParameters=true}
   /SetVariables(Particles.TempPosition:Vector3f=$Particles.FN_Light_Attributes.Position) {isSetParameters=true}
  ]
  rendererClasses [&NiagaraSpriteRendererProperties]
  renderers [
   {rendererClass=&NiagaraSpriteRendererProperties}
   {rendererClass=&NiagaraSpriteRendererProperties, rendererIndex=1}
   {rendererClass=&NiagaraSpriteRendererProperties, rendererIndex=2}
   {rendererClass=&NiagaraSpriteRendererProperties, rendererIndex=3}
  ]
 }
 Shockwave {
  emitterUpdateScript [
   &EmitterState(Life Cycle Mode=::NewEnumerator1, Loop Behavior=::NewEnumerator1, Loop Duration=1, UseLoopDelay=true)
   &SpawnBurst_Instantaneous(Use Spawn Probability=false, UseLoopCountLimit=false)
  ]
  particleSpawnScript [
   /Niagara/Modules/Spawn/Initialization/InitializeParticle(
    Color=&MultiplyLinearColor(
     Linear Color=<rgba 2 0.7603 0.2867 1.5>
     Scale Factor=&MakeFloat4FromVectorAndFloat("Vector (RGB)"=<1 1 1>, "Float (Alpha)"=&FN_FrontEndFloatModifier(Gameplay=0.5, Front End=0.5))
    )
    Lifetime=0.37
    Mesh Scale=&Multiply_VectorByFloat(
     Vector=&FN_FrontEndVector3Modifier(Gameplay=<0.5 2.1 2.1>, Front End=<0.4 1.5 1.5>)
     Float=&FN_FrontEndFloatModifier(Gameplay=0.2, Front End=0.2)
    )
    Module.Write Color=true
    Module.Write Lifetime=true
    Module.Write Mass=false
    Module.Write Material Random=true
    Module.Write Position=false
    Module.Write Scale=true
    Module.Write SpriteRotation=false
    Module.Write SpriteSize=false
    Position=&SimulationPosition()
   )
   &SystemLocation(MaskSpawnGroup=false)
   &InitialMeshOrientation(
    Orientation Axis=<0 0 1>
    Orientation Coordinate Space=::World
    Orientation Vector=<0 0 0>
    Rotation=&RandomRangeVector(Minimum=<0 0 0>, Maximum=<1 0 0>)
    Rotation Coordinate Space=::NewEnumerator1
    Use Orientation Vector=true
    Use Rotation Vector=false
   )
   &CameraOffset(Camera Offset Amount=&FN_FrontEndFloatModifier(Gameplay=30, Front End=0))
  ]
  particleUpdateScript [
   &ParticleState()
   &SolveForcesAndVelocity(Clamp Velocity=false, Limit Acceleration=false)
   &SkeletalMeshLocation(
    Bone Sampling Mode=::NewEnumerator4
    MaskSpawnGroup=false
    Orientation Sampling=::NewEnumerator1
    Skeletal Mesh=DataInterface() {propertyValues {
      DefaultMesh=&SuperSpike_Gadget
      ExcludeBoneName=root
      OptionalComponentClass=null
      PreviewMesh=&SuperSpike_Gadget
      SkinningMode=SkinOnTheFly
      SoftSourceActor=null
      SourceMode=Default
      UvSetIndex=0
      WholeMeshLOD=-1
      bExcludeBone=false
      bReadDeformedGeometry=true
      bRequireCurrentFrameData=true
      ComponentTags []
      FilteredBones [JNT_Ball_A]
      FilteredSockets []
      MeshUserParameter {parameter {name=None, typeDefHandle {registeredTypeIndex=1}, varData []}}
      SamplingRegions []
     }}
   )
   &ScaleColor(
    Scale Alpha=&Multiply_Float(
     A=1
     B=&Multiply_Float(
      A=&FloatFromCurve(FloatCurve=DataInterface() {propertyValues {Curve {keys [{arriveTangent=1.3571, interpMode=RCIM_Cubic, time=0.026, value=1}, <at 0.2 0.9>, <at 1 0>]}}})
      B=&FNTimeOfDayScalar(Night Scale=0.5, Morning Scale=0.2, Evening Scale=0.2)
     )
    )
    Scale RGB=&Multiply_VectorByFloat(
     Vector=<1 1 1>
     Float=&FloatFromCurve(FloatCurve=DataInterface() {propertyValues {Curve {keys [<at 0 1>,<at 1 0>]}}})
    )
    ScaleA=true
    ScaleRGB=true
   ) {module=ScaleColor001}
   &ScaleMeshSize(Scale Factor=&VectorFromCurve(
     VectorCurve=DataInterface() {propertyValues {
       XCurve {keys [
         {arriveTangent=4.1785, interpMode=RCIM_Cubic, leaveTangent=4.1785, tangentMode=RCTM_User, time=0, value=0}
         {arriveTangent=1.1179, interpMode=RCIM_Cubic, leaveTangent=1.1179, tangentMode=RCTM_User, time=0.2042, value=0.6228}
         {arriveTangent=0.0224, interpMode=RCIM_Cubic, leaveTangent=0.0224, tangentMode=RCTM_User, time=1, value=1.3}
        ]}
       YCurve {keys [
         {arriveTangent=4.1785, interpMode=RCIM_Cubic, leaveTangent=4.1785, tangentMode=RCTM_User, time=0, value=0}
         {arriveTangent=1.1179, interpMode=RCIM_Cubic, leaveTangent=1.1179, tangentMode=RCTM_User, time=0.2042, value=0.6228}
         {arriveTangent=0.0224, interpMode=RCIM_Cubic, leaveTangent=0.0224, tangentMode=RCTM_User, time=1, value=1}
        ]}
       ZCurve {keys [
         {arriveTangent=4.1785, interpMode=RCIM_Cubic, leaveTangent=4.1785, tangentMode=RCTM_User, time=0, value=0}
         {arriveTangent=1.1179, interpMode=RCIM_Cubic, leaveTangent=1.1179, tangentMode=RCTM_User, time=0.2042, value=0.6228}
         {arriveTangent=0.0224, interpMode=RCIM_Cubic, leaveTangent=0.0224, tangentMode=RCTM_User, time=1, value=1}
        ]}
      }}
     Scale Curve=<1 1.5 1.5>
    ))
   &DynamicMaterialParameters(
    Index 0 Param 1=&FloatFromCurve(FloatCurve=DataInterface() {propertyValues {Curve {keys [<at 0 0 RCIM_Cubic>,<at 1 0.28 RCIM_Cubic>]}}})
    Index 0 Param 2=&RandomRangeFloat(Minimum=0, Maximum=1)
    Index 0 Param 3=1000
    Index 0 Param 4=&FloatFromCurve(
     FloatCurve=DataInterface() {propertyValues {Curve {keys [<at 0 0>, {arriveTangent=-3.0374, leaveTangent=2.0371, time=0.2, value=0}, {interpMode=RCIM_Cubic, tangentMode=RCTM_Break, time=1, value=0}]}}}
     Scale Curve=0
    )
    Module.Param0WriteEnabled=true
    Module.Param1WriteEnabled=true
    Module.Param2WriteEnabled=true
    Module.Param3WriteEnabled=true
   )
  ]
  rendererClasses [&NiagaraMeshRendererProperties]
  renderers [{rendererClass=&NiagaraMeshRendererProperties, rendererIndex=0}]
 }
 Volleyball_Outlines {
  emitterUpdateScript [
   &EmitterState(Life Cycle Mode=::NewEnumerator1, Loop Behavior=::NewEnumerator1, Loop Duration=1, UseLoopDelay=false)
   &SpawnBurst_Instantaneous(Use Spawn Probability=false, UseLoopCountLimit=false)
  ]
  particleSpawnScript [
   /Niagara/Modules/Spawn/Initialization/V2/InitializeParticle(Color=<rgba 1 0.3607 0 1>, Lifetime=1.16, Mesh Scale Mode=::NewEnumerator3, UsePositionOffset=false, Write Material Random=true)
   &InitialMeshOrientation(Mesh Orientation Mode=::NewEnumerator1, Use Rotation Vector=false)
  ]
  particleUpdateScript [
   &ParticleState()
   &SkeletalMeshLocation(
    Bone Sampling Mode=::NewEnumerator4
    MaskSpawnGroup=false
    Position Sampling=::NewEnumerator2
    Skeletal Mesh=DataInterface() {propertyValues {
      DefaultMesh=&SuperSpike_Gadget
      ExcludeBoneName=root
      OptionalComponentClass=null
      PreviewMesh=&SuperSpike_Gadget
      SkinningMode=SkinOnTheFly
      SoftSourceActor=null
      SourceMode=Default
      UvSetIndex=0
      WholeMeshLOD=-1
      bExcludeBone=false
      bReadDeformedGeometry=true
      bRequireCurrentFrameData=true
      ComponentTags []
      FilteredBones [JNT_Ball_A]
      FilteredSockets []
      MeshUserParameter {parameter {name=None, typeDefHandle {registeredTypeIndex=1}, varData []}}
      SamplingRegions []
     }}
   )
   &ScaleMeshSize()
   /SetVariables(Particles.Position=&AddVectorToPosition(Position=$Particles.SkeletalMeshLocation.SampledPosition, Vector=<0 0 0>)) {isSetParameters=true}
  ]
  rendererClasses [&NiagaraMeshRendererProperties]
  renderers [{rendererClass=&NiagaraMeshRendererProperties, rendererIndex=0}]
 }
}
summary {
 systemName=NS_Emote_SuperSpike_Ball_Energy
 emitters [
  {emitterName=Shockwave, rendererClasses [&NiagaraMeshRendererProperties]}
  {emitterName=Flash, rendererClasses [&NiagaraSpriteRendererProperties]}
  {emitterName=Energy, rendererClasses [&NiagaraMeshRendererProperties, &NiagaraSpriteRendererProperties]}
  {emitterName=Volleyball_Outlines, rendererClasses [&NiagaraMeshRendererProperties]}
 ]
 userVariables [
  User.bIsFrontEnd {type=NiagaraBool}
  User.bIsFrontEndPreview {type=NiagaraBool}
  User.DustSpawnCurve {type=NiagaraFloat, defaultValue=1, description=""}
  User.Particle_BG_Color {type=LinearColor, description="", defaultValue {struct=LinearColor, value=<rgba 1 1 1 1>}}
  User.SkeletalMesh {
   type=NiagaraDataInterfaceSkeletalMesh
   defaultValue {
    struct=NiagaraExt_VariableValue_DataInterface
    value {dataInterface=&self:NiagaraDataInterfaceSkeletalMesh_0}
   }
  }
 ]
}
