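# Niagara system for the SuperSpike emote's ball ground impact, firing timed one-off bursts at a local offset.
# A layered glow flash, brief light, streaking spike sprites and expanding shockwave mesh, which bursts again later, accompany a fading crater decal.
# GPU rocks render as sprites and meshes in two spawn groups, one sampling the gadget's ball bone, with collision gated by quality level or front-end preview.
# Time-of-day inputs adjust shockwave color and spike opacity.
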NiagaraSystem /BRCosmetics/Animation/Game/MainPlayer/Emotes/SuperSpike/FX/NS_Emote_SuperSpike_Ball_Ground_Impact
dependencies {
 usedDataInterfaces [
  &NiagaraDataInterfaceMeshRendererInfo
  &NiagaraDataInterfaceCurve
  &NiagaraDataInterfacePlatformSet
  &NiagaraDataInterfaceColorCurve
  &NiagaraDataInterfaceVector2DCurve
  &NiagaraDataInterfaceCamera
  &NiagaraDataInterfaceVectorCurve
 ]
 usedDynamicInputs [&UniformRangedVector2D, &MakeFloatFromVector2d]
 usedModules [&SystemState]
}
emitters {
 Flash {
  emitterUpdateScript [
   &EmitterState(Life Cycle Mode=::NewEnumerator1, Loop Behavior=::NewEnumerator1, UseLoopDelay=false)
   &SpawnBurst_Instantaneous(Spawn Time=0.13, Use Spawn Probability=false, UseLoopCountLimit=false)
   /SetVariables(Emitter.Cam Pos:NiagaraPosition=&ReturnCameraVector(Camera Property=::NewEnumerator0, Coordinate Space=::Simulation)) {isSetParameters=true}
  ]
  particleSpawnScript [
   &InitializeParticle(
    Color=<rgba 1 0.4228 0.0194 0.1>
    Lifetime Max=0.25
    Lifetime Min=0.15
    Lifetime Mode=::NewEnumerator1
    Mass Max=2
    Mass Mode=::NewEnumerator3
    Position Offset=<-11.6492 135.0881 -90.1313>
    Position Offset Coordinate Space=::Local
    Sprite Rotation Mode=::NewEnumerator2
    Sprite Size Mode=::NewEnumerator3
    Uniform Sprite Size=&RandomRangeFloat(Minimum=100, Maximum=100)
    UsePositionOffset=true
    Write Material Random=true
   )
   /SetVariables(Particles.Position=&AddVectorToPosition(Position=$Particles.Position, Vector=<0 0 0>)) {isSetParameters=true}
   &SpriteFacingAndAlignment(Sprite Facing=<0 0 1>, WriteAlignment=false, WriteFacing=true)
  ]
  particleUpdateScript [
   &ParticleState()
   &CameraOffset(Camera Offset Amount=50) {module=CameraOffset001}
   &ScaleColor(
    Scale Alpha=&FloatFromCurve(
     FloatCurve=DataInterface() {propertyValues {Curve {keys [
         <at 0 0 RCIM_Cubic>
         {arriveTangent=0.1895, interpMode=RCIM_Cubic, leaveTangent=0.1895, tangentMode=RCTM_User, time=0.1, value=1}
         {arriveTangent=-0.0202, interpMode=RCIM_Cubic, leaveTangent=-0.0202, tangentMode=RCTM_User, time=0.7, value=1}
         <at 1.0011 0.002 RCIM_Cubic>
        ]}}}
     Scale Curve=10
    )
    ScaleA=true
    ScaleRGB=true
   )
   &ScaleSpriteSize(Scale Factor=&Vector2DFromFloat(Value=&FloatFromCurve(FloatCurve=DataInterface() {propertyValues {Curve {keys [
          {interpMode=RCIM_Cubic, tangentMode=RCTM_User, time=0, value=0}
          {arriveTangent=-0.0009, interpMode=RCIM_Cubic, leaveTangent=-0.0009, tangentMode=RCTM_User, time=0.2, value=1}
          {tangentMode=RCTM_User, time=1, value=0}
         ]}}})))
   /SetVariables(
    Particles.StarColor:LinearColor=&MultiplyLinearColor(
     Linear Color=$Particles.Color
     Scale Factor=&MakeVector4_FromLinearColor(LinearColor=<rgba 1 0.8483 0.1811 1>)
    )
    Particles.StarRotation:NiagaraFloat=&RandomRangeFloat(Minimum=0, Maximum=360)
    Particles.StarSize:Vector2f=&Multiply_Vector2(A=$Particles.SpriteSize, B=<2 2>)
   ) {isSetParameters=true}
   /SetVariables(
    Particles.StreaksColor:LinearColor=&MultiplyLinearColor(
     Linear Color=$Particles.Color
     Scale Factor=&MakeVector4_FromLinearColor(LinearColor=<rgba 1 0.5483 0 0.01>)
    )
    Particles.StreaksRotation:NiagaraFloat=&RandomRangeFloat(Minimum=0, Maximum=360)
    Particles.StreaksSize:Vector2f=&Multiply_Vector2(A=$Particles.SpriteSize, B=<7 7>)
   ) {isSetParameters=true}
   /SetVariables(
    Particles.AnamorphicColor:LinearColor=&MultiplyLinearColor(
     Linear Color=$Particles.Color
     Scale Factor=&MakeVector4_FromLinearColor(LinearColor=<rgba 1 0.3894 0 0.01>)
    )
    Particles.AnamorphicRotation=90
    Particles.AnamorphicSize:Vector2f=&Multiply_Vector2(A=$Particles.SpriteSize, B=<3 9>)
   ) {isSetParameters=true}
   /SetVariables(
    Particles.HaloColor:LinearColor=&MultiplyLinearColor(
     Linear Color=$Particles.Color
     Scale Factor=&MakeVector4_FromLinearColor(LinearColor=<rgba 1 0.2138 0 0.025>)
    )
    Particles.HaloSize:Vector2f=&Multiply_Vector2(A=$Particles.SpriteSize, B=<5 5>)
   ) {isSetParameters=true}
   /SetVariables(
    Particles.HaloColor_02:LinearColor=&MultiplyLinearColor(
     Linear Color=$Particles.Color
     Scale Factor=&MakeVector4_FromLinearColor(LinearColor=<rgba 1 0.2698 0 0.05>)
    )
    Particles.HaloSize_02:Vector2f=&Multiply_Vector2(A=$Particles.SpriteSize, B=<1.5 1.5>)
   ) {isSetParameters=true}
   /SetVariables(
    Particles.AirGlowColor:LinearColor=&MultiplyLinearColor(
     Linear Color=$Particles.Color
     Scale Factor=&MakeVector4_FromLinearColor(LinearColor=<rgba 1 0.8112 0.323 0.01>)
    )
    Particles.AirGlowSize:Vector2f=&Multiply_Vector2(A=$Particles.SpriteSize, B=<4 4>)
   ) {isSetParameters=true}
   /SetVariables(Particles.TempPosition:Vector3f=$Particles.FN_Light_Attributes.Position) {isSetParameters=true}
  ]
  rendererClasses [&NiagaraSpriteRendererProperties]
  renderers [
   {rendererClass=&NiagaraSpriteRendererProperties}
   {rendererClass=&NiagaraSpriteRendererProperties, rendererIndex=1}
   {rendererClass=&NiagaraSpriteRendererProperties, rendererIndex=2}
   {rendererClass=&NiagaraSpriteRendererProperties, rendererIndex=3}
  ]
 }
 Ground_Crater {
  emitterUpdateScript [
   &EmitterState(Life Cycle Mode=::NewEnumerator1, Loop Behavior=::NewEnumerator1, Loop Delay=0.5, Loop Duration=10, UseLoopDelay=false)
   &SpawnBurst_Instantaneous(Spawn Time=0.1, Use Spawn Probability=false, UseLoopCountLimit=false)
  ]
  particleSpawnScript [
   &DynamicMaterialParameters(
    Index 0 Param 1=&FloatFromCurve(
     FloatCurve=DataInterface() {propertyValues {Curve {defaultValue=1, keys [{tangentMode=RCTM_User, time=0, value=1}, <at 0.4 1>, {tangentMode=RCTM_User, time=1, value=1}]}}}
     CurveIndex=$Particles.Age
    )
    Module.Index0Float0Write=true
    Module.Index0Float1Write=true
    Module.Index0Float2Write=true
    Module.Index0Float3Write=true
   )
   &InitializeParticle(
    Color=<rgba 0.0046 0 1 1>
    Lifetime=&MakeCustomFloatFromBool(Bool=$User.bIsFrontEndPreview, True Float=1.5, False Float=1.5)
    Position Offset=<-11.6492 145.0881 -95.1313>
    Position Offset Coordinate Space=::Local
    Sprite Rotation Mode=::NewEnumerator2
    Sprite Size Mode=::NewEnumerator1
    Uniform Sprite Size Max=80
    Uniform Sprite Size Min=60
    UsePositionOffset=true
    Write Material Random=true
   ) {module=InitializeParticle001}
   /SetVariables(Particles.RandomDecalYZ:NiagaraFloat=&RandomRangeFloat(Minimum=150, Maximum=150)) {isSetParameters=true}
   /SetVariables(Particles.DecalSize:Vector3f=&MakeVector(X=75, Y=$Particles.RandomDecalYZ, Z=$Particles.RandomDecalYZ)) {isSetParameters=true}
   &InitialMeshOrientation(
    Mesh Orientation Mode=::NewEnumerator2
    Orientation Vector=<0 0 1>
    Rotation=&RandomRangeVector(Minimum=<-1 0 0>, Maximum=<1 0 0>)
    Use Rotation Vector=true
   )
   /SetVariables(Particles.VisibilityTag=&MakeCustomIntFromBool(Bool=$User.bIsFrontEnd, True INT=0, False INT=1)) {isSetParameters=true}
   &SpriteFacingAndAlignment(Sprite Alignment=<1 0 0>, Sprite Facing=<0 0 1>, WriteAlignment=true, WriteFacing=true) {module=SpriteFacingAndAlignment001}
   /Niagara/Modules/Spawn/Decal/InitialDecalOrientation(
    Orientation Vector:Vector3f=<0 0 1>
    Rotation:Vector3f=&RandomRangeVector(Minimum=<0 0 1>, Maximum=<0 0 1>)
    Use Rotation Vector=true
   )
  ]
  particleUpdateScript [
   &ParticleState()
   &ScaleSpriteSize(Uniform Curve Sprite Scale=DataInterface() {propertyValues {Curve {keys [
        {arriveTangent=230.8238, interpMode=RCIM_Cubic, leaveTangent=230.8238, tangentMode=RCTM_User, time=0, value=0}
        {arriveTangent=0.8135, interpMode=RCIM_Cubic, leaveTangent=0.8135, tangentMode=RCTM_User, time=0.01, value=1}
       ]}}})
   &ScaleColor(
    Scale Alpha=&FloatFromCurve(FloatCurve=DataInterface() {propertyValues {Curve {keys [
         {interpMode=RCIM_Cubic, tangentMode=RCTM_User, time=0, value=0}
         <at 0.025 1>
         {arriveTangent=-2.0513, interpMode=RCIM_Cubic, leaveTangent=-2.0513, time=0.5, value=1}
         <at 1 -1 RCIM_Cubic>
        ]}}})
    ScaleA=true
    ScaleRGB=true
   )
   &ScaleSpriteSize(Uniform Curve Sprite Scale=DataInterface() {propertyValues {Curve {keys [{arriveTangent=139.099, interpMode=RCIM_Cubic, leaveTangent=139.099, tangentMode=RCTM_User, time=0, value=0}, <at 0.015 1 RCIM_Cubic>]}}}) {module=ScaleSpriteSize001}
   /SetVariables(Particles.DecalFade:NiagaraFloat=&FloatFromCurve(
     FloatCurve=DataInterface() {propertyValues {Curve {keys [
         {interpMode=RCIM_Cubic, tangentMode=RCTM_User, time=0, value=0}
         <at 0.025 1>
         {arriveTangent=-1.0256, interpMode=RCIM_Cubic, leaveTangent=-1.0256, time=0.5, value=1}
         {arriveTangent=-1.72, interpMode=RCIM_Cubic, leaveTangent=-1.72, tangentMode=RCTM_User, time=1, value=0}
        ]}}}
     Scale Curve=&MakeCustomFloatFromBool(Bool=$User.bIsFrontEndPreview, True Float=0.5, False Float=1)
    )) {isSetParameters=true}
   /SetVariables(Particles.DecalSize:Vector3f=&Multiply_VectorByFloat(
     Vector=$Particles.Initial.DecalSize
     Float=&FloatFromCurve(FloatCurve=DataInterface() {propertyValues {Curve {keys [{arriveTangent=139.099, interpMode=RCIM_Cubic, leaveTangent=139.099, tangentMode=RCTM_User, time=0, value=0}, <at 0.015 1 RCIM_Cubic>]}}})
    )) {isSetParameters=true}
   &SolveForcesAndVelocity(Clamp Velocity=false, Limit Acceleration=false)
  ]
  rendererClasses [&NiagaraDecalRendererProperties]
  renderers [{rendererClass=&NiagaraDecalRendererProperties, rendererIndex=0}]
 }
 Light {
  emitterUpdateScript [
   &EmitterState(Life Cycle Mode=::NewEnumerator1, Loop Behavior=::NewEnumerator1, Loop Duration=2, UseLoopDelay=false)
   &SpawnBurst_Instantaneous(Spawn Time=0.1, Use Spawn Probability=false, UseLoopCountLimit=false)
  ]
  particleSpawnScript [&InitializeParticle(
    Color Mode=::NewEnumerator0
    Lifetime=0.15
    Position Offset=<-11.6492 135.0881 -95.1313>
    Position Offset Coordinate Space=::Local
    UsePositionOffset=true
    Write Material Random=true
   )]
  particleUpdateScript [
   &ParticleState()
   &FN_Light_Attributes(
    Color=<rgba 8 2.3947 1.293 1>
    Create Color Attribute=true
    Create Exponent Attribute=true
    Create Position Attribute=true
    Create Radius Attribute=true
    Exponent=3
    Position=<5 0 0>
    Radius=100
    Scale Color=&FloatFromCurve(FloatCurve=DataInterface() {propertyValues {Curve {keys [<at 0 0>,<at 0.1 1>,<at 0.9 1>,<at 1 0>]}}})
   )
  ]
  rendererClasses [&NiagaraLightRendererProperties]
  renderers [{rendererClass=&NiagaraLightRendererProperties, rendererIndex=0}]
 }
 Rocks {
  simTarget=GPUComputeSim
  emitterUpdateScript [
   &EmitterState(Life Cycle Mode=::NewEnumerator1, Loop Behavior=::NewEnumerator1, Loop Duration=2, UseLoopDelay=false)
   &SpawnBurst_Instantaneous(Spawn Count=7, Spawn Time=0.13, Use Spawn Probability=false, UseLoopCountLimit=false)
   &SpawnBurst_Instantaneous(Spawn Count=5, Spawn Group=1, Spawn Time=0.13, Use Spawn Probability=false, UseLoopCountLimit=false) {module=SpawnBurst_Instantaneous001}
  ]
  particleSpawnScript [
   &InitializeParticle(
    Color=<rgba 0.0505 0.053 0.0501 1>
    Lifetime=&UniformRangedFloat(Minimum=0.5, Maximum=2)
    Mesh Renderer Array Visibility Mode=::NewEnumerator3
    Mesh Renderer Info=DataInterface() {propertyValues {MeshRenderer=&self:Rocks_0.NiagaraMeshRendererProperties_0}}
    Mesh Scale Max=<0.5 0.5 0.5>
    Mesh Scale Min=<0.2 0.2 0.5>
    Mesh Scale Mode=::NewEnumerator2
    Position Offset=&MaskVectorBySpawnGroup(Masked Vector=<0 0 0>, Passthrough Vector=<-11.6492 135.0881 -95.1313>)
    Position Offset Coordinate Space=::Local
    Sprite Rotation Mode=::NewEnumerator2
    Sprite Size Max=<3 3>
    Sprite Size Min=<1.5 1.5>
    Sprite Size Mode=::NewEnumerator2
    UsePositionOffset=true
    Write Material Random=true
   ) {module=InitializeParticle001}
   &SkeletalMeshLocation(
    Bone Sampling Mode=::NewEnumerator4
    MaskSpawnGroup=false
    Orientation Sampling=::NewEnumerator1
    Position Sampling=::NewEnumerator2
    Skeletal Mesh=DataInterface() {propertyValues {
      DefaultMesh=null
      ExcludeBoneName=root
      OptionalComponentClass=null
      PreviewMesh=/BRCosmetics/Animation/Props/Gadgets_SuperSpike/Meshes/SuperSpike_Gadget
      SkinningMode=SkinOnTheFly
      SoftSourceActor=null
      SourceMode=Default
      UvSetIndex=0
      WholeMeshLOD=-1
      bExcludeBone=false
      bReadDeformedGeometry=true
      bRequireCurrentFrameData=false
      ComponentTags []
      FilteredBones [JNT_Ball_A]
      FilteredSockets []
      MeshUserParameter {parameter {name=None, typeDefHandle {registeredTypeIndex=1}, varData []}}
      SamplingRegions []
     }}
   )
   /SetVariables(Particles.Position=&MaskVectorBySpawnGroup(
     Masked Vector=Linked() {linkedVariable {name=Particles.SkeletalMeshLocation.SampledPosition, type=NiagaraPosition}}
     Passthrough Vector=Linked() {linkedVariable {name=Particles.Position, type=NiagaraPosition}}
    )) {isSetParameters=true}
   /SetVariables(Particles.RockSize:NiagaraFloat=&RandomRangeFloat(Minimum=0.4, Maximum=1.4)) {isSetParameters=true}
   /SetVariables(Particles.SpriteSize=&Multiply_Vector2D_ByFloat(Float=$Particles.RockSize, Vector2D=$Particles.SpriteSize)) {isSetParameters=true}
   /SetVariables(Particles.Scale=&Multiply_VectorByFloat(Vector=$Particles.Scale, Float=$Particles.RockSize)) {isSetParameters=true}
   &ShapeLocation(
    MaskSpawnGroup=false
    Offset Mode=::NewEnumerator0
    Override Seed=false
    Ring Radius=20
    Shape Primitive=::NewEnumerator4
    UseOwnerScale=true
   )
   &AddVelocityFromPoint(
    Module.Use Falloff Distance=false
    Origin Offset=&MaskVectorBySpawnGroup(Masked Vector=<0 0 -1>, Passthrough Vector=<-11.6492 135.0881 -95.1313>)
    Velocity Strength=&Multiply_Float(
     A=&MaskFloatBySpawnGroup(Masked Float=200, Passthrough Float=100)
     B=&FN_FrontEndFloatModifier(Gameplay=1, Front End=1)
    )
   )
   &AddVelocityInCone(
    Cone Angle=170
    Cone Axis=&RotateVectorByQuaternion(Vector To Rotate=<0 0 1>, Quaternion=<0 0 0 1>)
    Cone Axis Coordinate Space=::Simulation
    Module.Use Velocity Falloff On Cone Axis=true
    Velocity Distribution Along Cone Axis=1
    Velocity Falloff Away From Cone Axis=0.333
    Velocity Strength=&Multiply_Float(
     A=&MaskFloatBySpawnGroup(Masked Float=&UniformRangedFloat(Minimum=50, Maximum=125), Passthrough Float=0)
     B=&FN_FrontEndFloatModifier(Gameplay=1, Front End=1)
    )
   ) {module=AddVelocityInCone001}
   &AccelerationForce(Acceleration=&Multiply_Vector(
     A=&MaskVectorBySpawnGroup(
      Masked Vector=&UniformRangedVector(
       Minimum=<0 0 -10>
       Maximum=&MakeVector(
        X=0
        Y=0
        Z=&FloatFromCurve(
         FloatCurve=DataInterface() {propertyValues {Curve {keys [<at 0.0174 1.2>,<at 0.9841 0.2965>]}}}
         CurveIndex=$Emitter.NormalizedLoopAge
         Scale Curve=45
        )
       )
      )
      Passthrough Vector=<0 0 0>
     )
     B=&VectorFromFloat(Value=&FN_FrontEndFloatModifier(Gameplay=1, Front End=1))
    )) {module=AccelerationForce001}
   &AddRotationalVelocity(Rotation Rate=&RandomVector(Vector Scale=1))
   &AddVelocity(
    Velocity=&Multiply_Vector(
     A=&Multiply_Vector(
      A=&RandomRangeVector(Minimum=<0 0 130>, Maximum=<0 0 190>)
      B=&VectorFromFloat(Value=&MaskFloatBySpawnGroup(Masked Float=2, Passthrough Float=1))
     )
     B=<1 1 1>
    )
    Velocity Speed Scale=&FN_FrontEndFloatModifier(Gameplay=1, Front End=1)
   )
   &PointForce(
    Force Origin=&SimulationPosition()
    Force Strength=&UniformRangedFloat(Minimum=10, Maximum=&Multiply_Float(A=20, B=1))
    Module.Use Falloff=false
    Module.UseLinearForce=false
   ) {enabled=false}
   /SetVariables(Particles.Velocity=&Multiply_VectorByFloat(
     Vector=$Particles.Velocity
     Float=&FloatFromCurve(FloatCurve=DataInterface() {propertyValues {Curve {keys [<at 0.4 2>,<at 1.4 1>]}}}, CurveIndex=$Particles.RockSize)
    )) {isSetParameters=true}
   &ApplyInitialForces()
   &AddVelocityInCone(
    Cone Axis=<0 0 1>
    Module.Use Velocity Falloff On Cone Axis=false
    Velocity Strength=&UniformRangedFloat(Minimum=100, Maximum=400)
   ) {enabled=false}
   /Niagara/Modules/Update/SubUV/SubUVAnimation(Random SubUV Lookup Index:NiagaraFloat=$Particles.RandomFloat, SubUV Animation Mode=::NewEnumerator1) {module=SubUVAnimation001}
   /SetVariables(Particles.VisibilityTag=&MakeCustomIntFromBool(
     Bool=&Boolean_OrOperation(
      Bool1=&IsPlatformSetActive(New Platform Set=DataInterface() {propertyValues {Platforms {qualityLevelMask=-2}}})
      Bool2=$User.bIsFrontEndPreview
     )
     True INT=&RandomRangeInt(Minimum=0, Maximum=1)
     False INT=0
    )) {isSetParameters=true}
  ]
  particleUpdateScript [
   &ParticleState() {module=UpdateAge}
   &Drag(Use Linear Drag=true, Use Rotational Drag=false)
   &GravityForce()
   &Collision(
    Advanced Aging Rate=1
    Age Colliding Particles=false
    Collision Enabled=&Boolean_OrOperation(
     Bool1=&IsPlatformSetActive(New Platform Set=DataInterface() {propertyValues {Platforms {qualityLevelMask=-4}}})
     Bool2=$User.bIsFrontEndPreview
    )
    Control Roll On Collision=true
    Enable Rest State=false
    GPU Collision Type=::NewEnumerator2
    Module.EnableMaxCollisionCount=false
    Module.Randomize Collision Normal Vector=true
    Particle Radius Scale=0.5
    Restitution=0.4
    ScaleRateOfRotation=false
   )
   &SpriteRotationRate(Rotation Rate=&Multiply_Float(A=$Particles.RandomFloat, B=4))
   &ScaleSpriteSize(
    Uniform Curve Scale=2
    Uniform Curve Sprite Scale=DataInterface() {propertyValues {Curve {keys [
        <at 0 1>
        {arriveTangent=-1, interpMode=RCIM_Cubic, leaveTangent=-1, time=0.8, value=1}
        {arriveTangent=-5.7307, interpMode=RCIM_Cubic, leaveTangent=-5.7307, tangentMode=RCTM_User, time=1, value=0}
       ]}}}
   )
   &ScaleMeshSize(Scale Factor=&VectorFromFloat(Value=&FloatFromCurve(FloatCurve=DataInterface() {propertyValues {Curve {keys [
          <at 0 1>
          {arriveTangent=-1, interpMode=RCIM_Cubic, leaveTangent=-1, time=0.8, value=1}
          {arriveTangent=-5.7307, interpMode=RCIM_Cubic, leaveTangent=-5.7307, tangentMode=RCTM_User, time=1, value=0}
         ]}}})))
   &SpriteRotationRate(Rotation Rate=&RandomRangeFloat(Minimum=-300, Maximum=300)) {module=SpriteRotationRate001}
   &SampleSkeletalMesh(
    AdditionalYawPitchRoll=false
    Mesh Sampling Type:ENiagaraSkelSamplingModeFull=::NewEnumerator5
    Sampled Orientation Coordinate Space:ENiagaraCoordinateSpace=::World
    Skeletal Mesh=DataInterface() {propertyValues {
      DefaultMesh=null
      ExcludeBoneName=root
      OptionalComponentClass=null
      PreviewMesh=/Game/Characters/Player/Male/Medium/Base/SK_M_MALE_Base_Skeleton
      SkinningMode=SkinOnTheFly
      SoftSourceActor=null
      SourceMode=Default
      UvSetIndex=0
      WholeMeshLOD=-1
      bExcludeBone=false
      bReadDeformedGeometry=true
      bRequireCurrentFrameData=true
      ComponentTags []
      FilteredBones []
      FilteredSockets [Root_Effect]
      MeshUserParameter {parameter {name=None, typeDefHandle {registeredTypeIndex=1}, varData []}}
      SamplingRegions []
     }}
    Socket Sampling Mode:ENiagaraSocketSamplingMode=::NewEnumerator5
   ) {module=SampleSkeletalMesh001}
   &Color(Color=&ColorFromCurve(ColorCurve=DataInterface() {propertyValues {
       AlphaCurve {keys [<at 0 1 RCIM_Cubic>,<at 1 1 RCIM_Cubic>]}
       BlueCurve {keys [<at 0 0.0865 RCIM_Cubic>,<at 1 1 RCIM_Cubic>]}
       GreenCurve {keys [<at 0 0.0814 RCIM_Cubic>,<at 1 1 RCIM_Cubic>]}
       RedCurve {keys [<at 0 0.8854 RCIM_Cubic>,<at 1 1 RCIM_Cubic>]}
      }}))
   &SolveForcesAndVelocity(Clamp Velocity=false, Limit Acceleration=false)
   /SetVariables(Particles.Position=&MakeCustomVectorFromBool(
     Bool=$User.bIsFrontEndPreview
     True Vector=&AddVectorToPosition(
      Position=$Particles.Position
      Vector=&Multiply_VectorByFloat(
       Vector=&RotateVectorByQuaternion(Vector To Rotate=<1 0 0>, Quaternion=$StackContext.SampleSkeletalMesh001.SampledRotation)
       Float=-5
      )
     )
     False Vector=Linked() {linkedVariable {name=Particles.Position, type=NiagaraPosition}}
    )) {enabled=false, isSetParameters=true}
  ]
  rendererClasses [&NiagaraSpriteRendererProperties, &NiagaraMeshRendererProperties]
  renderers [
   {rendererClass=&NiagaraSpriteRendererProperties}
   {rendererClass=&NiagaraMeshRendererProperties, rendererIndex=1}
  ]
 }
 Shockwave {
  emitterUpdateScript [
   &EmitterState(Life Cycle Mode=::NewEnumerator1, Loop Behavior=::NewEnumerator1, UseLoopDelay=false)
   &SpawnBurst_Instantaneous(Spawn Time=0.13, Use Spawn Probability=false, UseLoopCountLimit=false)
   &SpawnBurst_Instantaneous(Spawn Time=3.1667, Use Spawn Probability=false, UseLoopCountLimit=false) {module=SpawnBurst_Instantaneous001}
  ]
  particleSpawnScript [
   &InitializeParticle(
    Color=&FNTimeOfDayColor(Night Color=<rgba 0.2039 0.2396 1 0.8>, Morning Color=<rgba 1 1 1 0.8>, Day Color=<rgba 1 1 1 0.8>, Evening Color=<rgba 1 1 1 0.8>)
    Lifetime=0.5
    Mesh Scale Mode=::NewEnumerator3
    Mesh Uniform Scale=0.5
    Position Offset=<-11.6492 135.0881 -200.1313>
    Position Offset Coordinate Space=::Local
    UsePositionOffset=true
    Write Material Random=true
   )
   &SkeletalMeshLocation(
    Bone Sampling Mode=::NewEnumerator4
    MaskSpawnGroup=false
    Orientation Sampling=::NewEnumerator1
    Sampled Position Offset=<-11.6492 135.0881 -60>
    Sampled Position Offset Coordinate Space:ENiagaraSkelMeshTransforms=::NewEnumerator2
    Skeletal Mesh=DataInterface() {propertyValues {
      DefaultMesh=null
      ExcludeBoneName=root
      OptionalComponentClass=null
      PreviewMesh=&SuperSpike_Gadget
      SkinningMode=SkinOnTheFly
      SoftSourceActor=null
      SourceMode=Default
      UvSetIndex=0
      WholeMeshLOD=-1
      bExcludeBone=false
      bReadDeformedGeometry=true
      bRequireCurrentFrameData=false
      ComponentTags []
      FilteredBones [root]
      FilteredSockets []
      MeshUserParameter {parameter {name=None, typeDefHandle {registeredTypeIndex=1}, varData []}}
      SamplingRegions []
     }}
   )
   &InitialMeshOrientation(Mesh Orientation Mode=::NewEnumerator2, Use Rotation Vector=false)
  ]
  particleUpdateScript [
   &ParticleState()
   &UpdateMeshOrientation(Rotation Vector=<0 0 1>)
   &DynamicMaterialParameters(
    Index 0 Param 1=&FloatFromCurve(FloatCurve=DataInterface() {propertyValues {Curve {keys [{interpMode=RCIM_Cubic, tangentMode=RCTM_User, time=0.5, value=0}, {interpMode=RCIM_Cubic, tangentMode=RCTM_User, time=1, value=1}]}}})
    Index 0 Param 3=1
    Module.Param0WriteEnabled=true
    Module.Param1WriteEnabled=true
    Module.Param2WriteEnabled=true
    Module.Param3WriteEnabled=true
   )
   &ScaleMeshSize(Scale Factor=&VectorFromCurve(
     VectorCurve=DataInterface() {propertyValues {
       XCurve {keys [
         {arriveTangent=4.1785, interpMode=RCIM_Cubic, leaveTangent=4.1785, tangentMode=RCTM_User, time=0, value=0}
         {arriveTangent=1.1179, interpMode=RCIM_Cubic, leaveTangent=1.1179, tangentMode=RCTM_User, time=0.2042, value=0.6228}
         {arriveTangent=0.0224, interpMode=RCIM_Cubic, leaveTangent=0.0224, tangentMode=RCTM_User, time=1, value=1}
        ]}
       YCurve {keys [
         {arriveTangent=4.1785, interpMode=RCIM_Cubic, leaveTangent=4.1785, tangentMode=RCTM_User, time=0, value=0}
         {arriveTangent=1.1179, interpMode=RCIM_Cubic, leaveTangent=1.1179, tangentMode=RCTM_User, time=0.2042, value=0.6228}
         {arriveTangent=0.0224, interpMode=RCIM_Cubic, leaveTangent=0.0224, tangentMode=RCTM_User, time=1, value=1}
        ]}
       ZCurve {keys [{tangentMode=RCTM_User, time=0, value=1}, {tangentMode=RCTM_User, time=1, value=1}]}
      }}
     Scale Curve=<3 3 2>
    ))
   &ScaleColor(
    Linear Color Curve=DataInterface() {propertyValues {
      AlphaCurve {keys [<at 0 0>,<at 0.2021 0.8>,<at 1 0>]}
      BlueCurve {keys [<at 0 0.1682>,<at 0.1795 0.405>,<at 0.6246 0.3697>,<at 1 0.1357>]}
      GreenCurve {keys [<at 0 0.192>,<at 0.1795 0.6739>,<at 0.6246 0.5905>,<at 1 0.1211>]}
      RedCurve {keys [<at 0 0.3229>,<at 0.1795 0.9844>,<at 0.6246 0.9219>,<at 1 0.3021>]}
     }}
    Scale Mode=::NewEnumerator2
   ) {module=ScaleColor001}
  ]
  rendererClasses [&NiagaraMeshRendererProperties]
  renderers [{rendererClass=&NiagaraMeshRendererProperties, rendererIndex=0}]
 }
 Spikes {
  emitterUpdateScript [
   &EmitterState(Life Cycle Mode=::NewEnumerator1, Loop Behavior=::NewEnumerator1, Loop Duration=2, UseLoopDelay=true)
   &SpawnBurst_Instantaneous(Spawn Count=20, Spawn Time=0.1, Use Spawn Probability=false, UseLoopCountLimit=false)
  ]
  particleSpawnScript [
   &InitializeParticle(
    Color=<rgba 1 1 1 0.35>
    Lifetime Max=0.25
    Lifetime Min=0.15
    Lifetime Mode=::NewEnumerator1
    Position Offset=<-11.6492 135.0881 -95.1313>
    Position Offset Coordinate Space=::Local
    Sprite Size Max=<50 100>
    Sprite Size Min=<10 100>
    Sprite Size Mode=::NewEnumerator2
    UsePositionOffset=true
    Write Material Random=true
   )
   &TorusLocation(Align Torus With Axis=false, Handle Radius=0, Large Radius=1, MaskSpawnGroup=false, Module.Override Randomness=false)
   &AddVelocityFromPoint(
    Module.Use Falloff Distance=false
    Origin Offset=<-11.6492 135.0881 -95.1313>
    Velocity Strength=&RandomRangeFloat(Minimum=100, Maximum=800)
   )
   &AddVelocity(Velocity=&Multiply_Vector(
     A=&RandomRangeVector(Minimum=<0 0 130>, Maximum=<0 0 190>)
     B=&VectorFromFloat(Value=1)
    ))
   /Niagara/Modules/Update/SubUV/V2/SubUVAnimation(End Frame=0, SubUV Animation Mode:ENiagara_SubUVLookupMode=::NewEnumerator2)
   &CameraOffset(Camera Offset Amount=10)
  ]
  particleUpdateScript [
   &ParticleState()
   &Drag(Drag=6, Use Linear Drag=true, Use Rotational Drag=false)
   &SolveForcesAndVelocity(Clamp Velocity=false, Limit Acceleration=false)
   &ScaleSpriteSize(
    Non-Uniform Curve Scale=&Multiply_Vector2D_ByFloat(Float=0.6, Vector2D=<1 1>)
    Non-Uniform Curve Sprite Scale=DataInterface() {propertyValues {
      XCurve {keys [{tangentMode=RCTM_User, time=0, value=1}, {tangentMode=RCTM_User, time=1, value=0.8}]}
      YCurve {keys [{tangentMode=RCTM_User, time=0, value=0.5}, {tangentMode=RCTM_User, time=1, value=1.7}]}
     }}
    Scale Sprite Size Mode=::NewEnumerator3
   )
   &ScaleColor(
    Scale Alpha=&FloatFromCurve(
     FloatCurve=DataInterface() {propertyValues {Curve {keys [<at 0 0>,<at 0.1 1>,<at 0.2 0.8>,<at 1 0>]}}}
     Scale Curve=&FNTimeOfDayScalar(Night Scale=1, Morning Scale=0.3, Evening Scale=0.5)
    )
    ScaleA=true
    ScaleRGB=true
   )
   &ScaleColor(
    Linear Color Curve=DataInterface() {propertyValues {
      AlphaCurve {keys [<at 0 1>,<at 1 0>]}
      BlueCurve {keys [<at 0 0.0011>,<at 0.5077 0>,<at 0.5323 0>,<at 0.559 0.054>,<at 0.5836 0.0248>,<at 1 0.1667>]}
      GreenCurve {keys [<at 0 0.3102>,<at 0.5077 0.372>,<at 0.5323 0.1259>,<at 0.559 0>,<at 0.5836 0.0282>,<at 1 0>]}
      RedCurve {keys [<at 0 1>,<at 0.5077 0.8021>,<at 0.5323 0.5749>,<at 0.559 0.6197>,<at 0.5836 0.3177>,<at 1 0.0742>]}
     }}
    Scale Mode=::NewEnumerator2
   ) {module=ScaleColor001}
  ]
  rendererClasses [&NiagaraSpriteRendererProperties]
  renderers [{rendererClass=&NiagaraSpriteRendererProperties}]
 }
}
summary {
 systemName=NS_Emote_SuperSpike_Ball_Ground_Impact
 emitters [
  {
   emitterName=Rocks
   simTarget=GPUComputeSim
   rendererClasses [&NiagaraSpriteRendererProperties, &NiagaraMeshRendererProperties]
  }
  {emitterName=Spikes, rendererClasses [&NiagaraSpriteRendererProperties]}
  {emitterName=Light, rendererClasses [&NiagaraLightRendererProperties]}
  {emitterName=Ground_Crater, rendererClasses [&NiagaraDecalRendererProperties]}
  {emitterName=Flash, rendererClasses [&NiagaraSpriteRendererProperties]}
  {emitterName=Shockwave, rendererClasses [&NiagaraMeshRendererProperties]}
 ]
 userVariables [
  User.bIsFrontEnd {type=NiagaraBool}
  User.bIsFrontEndPreview {type=NiagaraBool}
  User.DustSpawnCurve {type=NiagaraFloat, defaultValue=1, description=""}
  User.Particle_BG_Color {type=LinearColor, description="", defaultValue {struct=LinearColor, value=<rgba 1 1 1 1>}}
  User.SkeletalMesh {
   type=NiagaraDataInterfaceSkeletalMesh
   defaultValue {
    struct=NiagaraExt_VariableValue_DataInterface
    value {dataInterface=&self:NiagaraDataInterfaceSkeletalMesh_0}
   }
  }
 ]
}
