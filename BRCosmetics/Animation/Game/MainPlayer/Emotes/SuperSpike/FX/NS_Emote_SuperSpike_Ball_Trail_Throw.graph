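# Niagara system for the SuperSpike emote's volleyball throw, adding a fiery trail and glow to the ball.
# Every emitter samples the gadget mesh's JNT_Ball_A bone to follow the ball.
# Tapering dual-renderer ribbons, fire sprites, a time-of-day-tinted background sprite, a light and a mesh outline form the effect; a Glow emitter is disabled.
# Front-end modifiers adjust color and opacity, hiding the background sprite outside gameplay.
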NiagaraSystem /BRCosmetics/Animation/Game/MainPlayer/Emotes/SuperSpike/FX/NS_Emote_SuperSpike_Ball_Trail_Throw
dependencies {
 usedDataInterfaces [&NiagaraDataInterfaceCurve, &NiagaraDataInterfaceCamera, &NiagaraDataInterfaceColorCurve]
 usedDynamicInputs []
 usedModules [&SystemState]
}
emitters {
 BG {
  emitterUpdateScript [
   &EmitterState(Inactive Response=::NewEnumerator1, Life Cycle Mode=::NewEnumerator1, Loop Behavior=::NewEnumerator1, UseLoopDelay=false)
   &SpawnBurst_Instantaneous(Spawn Time=0.01, Use Spawn Probability=false, UseLoopCountLimit=false)
   /SetVariables(Emitter.Cam Pos:NiagaraPosition=&ReturnCameraVector(Camera Property=::NewEnumerator0, Coordinate Space=::Simulation)) {isSetParameters=true}
  ]
  particleSpawnScript [
   &InitializeParticle(
    Color=&FNTimeOfDayColor(
     Night Color=<rgba 0.1198 0.0271 0.004 0.05>
     Morning Color=<rgba 0.1198 0.0271 0.004 0.05>
     Day Color=<rgba 0.6406 0.1648 0.0868 0.05>
     Evening Color=<rgba 0.1198 0.0271 0.004 0.05>
    )
    Lifetime=0.6
    Mass Max=2
    Mass Mode=::NewEnumerator3
    Position Offset Coordinate Space=::Local
    Sprite Rotation Mode=::NewEnumerator2
    Sprite Size Mode=::NewEnumerator3
    Uniform Sprite Size=30
    UsePositionOffset=true
    Write Material Random=true
   )
   &SkeletalMeshLocation(
    Bone Sampling Mode=::NewEnumerator4
    MaskSpawnGroup=false
    Orientation Sampling=::NewEnumerator1
    Skeletal Mesh=DataInterface() {propertyValues {
      DefaultMesh=/BRCosmetics/Animation/Props/Gadgets_SuperSpike/Meshes/SuperSpike_Gadget
      ExcludeBoneName=root
      OptionalComponentClass=null
      PreviewMesh=&SuperSpike_Gadget
      SkinningMode=SkinOnTheFly
      SoftSourceActor=null
      SourceMode=Default
      UvSetIndex=0
      WholeMeshLOD=-1
      bExcludeBone=false
      bReadDeformedGeometry=true
      bRequireCurrentFrameData=true
      ComponentTags []
      FilteredBones [JNT_Ball_A]
      FilteredSockets []
      MeshUserParameter {parameter {name=None, typeDefHandle {registeredTypeIndex=1}, varData []}}
      SamplingRegions []
     }}
   )
   /SetVariables(Particles.Position=&AddVectorToPosition(Position=$Particles.Position, Vector=<0 0 0>)) {isSetParameters=true}
  ]
  particleUpdateScript [
   &ScaleColor(Scale Alpha=&FN_FrontEndFloatModifier(Gameplay=1, Front End=0), ScaleA=true, ScaleRGB=true) {module=ScaleColor001}
   &ParticleState()
   &CameraOffset() {module=CameraOffset001}
   &SkeletalMeshLocation(
    Bone Sampling Mode=::NewEnumerator4
    MaskSpawnGroup=false
    Orientation Sampling=::NewEnumerator1
    Skeletal Mesh=DataInterface() {propertyValues {
      DefaultMesh=&SuperSpike_Gadget
      ExcludeBoneName=root
      OptionalComponentClass=null
      PreviewMesh=&SuperSpike_Gadget
      SkinningMode=SkinOnTheFly
      SoftSourceActor=null
      SourceMode=Default
      UvSetIndex=0
      WholeMeshLOD=-1
      bExcludeBone=false
      bReadDeformedGeometry=true
      bRequireCurrentFrameData=true
      ComponentTags []
      FilteredBones [JNT_Ball_A]
      FilteredSockets []
      MeshUserParameter {parameter {name=None, typeDefHandle {registeredTypeIndex=1}, varData []}}
      SamplingRegions []
     }}
   ) {module=SkeletalMeshLocation001}
   &ScaleColor(
    Scale Alpha=&FloatFromCurve(
     FloatCurve=DataInterface() {propertyValues {Curve {keys [
         {arriveTangent=-0.0202, interpMode=RCIM_Cubic, leaveTangent=-0.0202, tangentMode=RCTM_User, time=0.8, value=1}
         {arriveTangent=-4.9617, interpMode=RCIM_Cubic, leaveTangent=-4.9617, time=0.8, value=1}
         <at 1.0011 0.002 RCIM_Cubic>
        ]}}}
     Scale Curve=10
    )
    ScaleA=true
    ScaleRGB=true
   )
   /SetVariables(
    Particles.AirGlowColor:LinearColor=&MultiplyLinearColor(Linear Color=$Particles.Color, Scale Factor=<1 1 1 1>)
    Particles.AirGlowSize:Vector2f=&Multiply_Vector2(A=$Particles.SpriteSize, B=<3 3>)
   ) {isSetParameters=true}
   /SetVariables(Particles.TempPosition:Vector3f=$Particles.FN_Light_Attributes.Position) {isSetParameters=true}
  ]
  rendererClasses [&NiagaraSpriteRendererProperties]
  renderers [{rendererClass=&NiagaraSpriteRendererProperties}]
 }
 Fire {
  emitterUpdateScript [
   &EmitterState(Life Cycle Mode=::NewEnumerator1, Loop Behavior=::NewEnumerator1, Loop Duration=0.6, UseLoopDelay=true)
   &SpawnRate(SpawnRate=150, Use Spawn Probability=false)
  ]
  particleSpawnScript [
   &InitializeParticle(
    Color Maximum=<rgba 1 0.7857 0.0865 1>
    Color Minimum=<rgba 1 0.3168 0.079 1>
    Color Mode=::NewEnumerator2
    Lifetime=&RandomRangeFloat(Minimum=0.15, Maximum=0.25)
    Sprite Rotation Mode=::NewEnumerator2
    Sprite Size=&Multiply_Vector2D_ByFloat(
     Float=&Multiply_Float(
      A=&RandomRangeFloat(Minimum=160, Maximum=200)
      B=&Multiply_Float(
       A=0.35
       B=&FloatFromCurve(
        FloatCurve=DataInterface() {propertyValues {Curve {defaultValue=1, keys [{tangentMode=RCTM_User, time=0, value=0.75}, {tangentMode=RCTM_User, time=1, value=1}]}}}
        CurveIndex=$Emitter.NormalizedLoopAge
       )
      )
     )
     Vector2D=<0.8 1>
    )
    Sprite Size Mode=::NewEnumerator4
    UsePositionOffset=false
    Write Material Random=true
   )
   &ShapeLocation(
    MaskSpawnGroup=false
    Offset=&Multiply_VectorByFloat(Vector=<-60 0 0>, Float=$System.FX_Scale)
    Offset Mode=::NewEnumerator0
    Override Seed=false
    Rotation Mode=::NewEnumerator1
    Sphere Radius=&Multiply_Float(A=0.5, B=$System.FX_Scale)
    UseOwnerScale=true
   )
   &CameraOffset(Camera Offset Amount=-10)
  ]
  particleUpdateScript [
   &ScaleColor(Scale Alpha=&FN_FrontEndFloatModifier(Gameplay=1, Front End=0.8), ScaleA=true, ScaleRGB=true) {module=ScaleColor004}
   &Drag(Use Linear Drag=true, Use Rotational Drag=true)
   &SkeletalMeshLocation(
    Bone Sampling Mode=::NewEnumerator4
    MaskSpawnGroup=false
    Orientation Sampling=::NewEnumerator1
    Skeletal Mesh=DataInterface() {propertyValues {
      DefaultMesh=&SuperSpike_Gadget
      ExcludeBoneName=root
      OptionalComponentClass=null
      PreviewMesh=&SuperSpike_Gadget
      SkinningMode=SkinOnTheFly
      SoftSourceActor=null
      SourceMode=Default
      UvSetIndex=0
      WholeMeshLOD=-1
      bExcludeBone=false
      bReadDeformedGeometry=true
      bRequireCurrentFrameData=true
      ComponentTags []
      FilteredBones [JNT_Ball_A]
      FilteredSockets []
      MeshUserParameter {parameter {name=None, typeDefHandle {registeredTypeIndex=1}, varData []}}
      SamplingRegions []
     }}
   )
   &SolveForcesAndVelocity(Clamp Velocity=false, Limit Acceleration=false)
   &ParticleState()
   &ScaleSpriteSize(Uniform Curve Sprite Scale=DataInterface() {propertyValues {Curve {keys [
        {arriveTangent=0.2827, interpMode=RCIM_Cubic, leaveTangent=0.2827, tangentMode=RCTM_User, time=0, value=1}
        {interpMode=RCIM_Cubic, tangentMode=RCTM_User, time=1, value=0.6}
       ]}}})
   &ScaleColor(
    Scale Alpha=&FloatFromCurve(
     FloatCurve=DataInterface() {propertyValues {Curve {keys [
         {interpMode=RCIM_Cubic, tangentMode=RCTM_User, time=0, value=0}
         {interpMode=RCIM_Cubic, tangentMode=RCTM_User, time=0.4, value=1}
         {arriveTangent=-2.1237, interpMode=RCIM_Cubic, leaveTangent=-2.1237, tangentMode=RCTM_User, time=1, value=0}
        ]}}}
     Scale Curve=&Multiply_Float(A=0.4, B=1)
    )
    Scale RGB=&MakeVectorFromLinearColorRGB(Color=&ColorFromCurve(ColorCurve=DataInterface() {propertyValues {
        AlphaCurve {keys [<at 0 1 RCIM_Cubic>]}
        BlueCurve {keys [<at 0.0938 0.2527>,<at 0.5194 0 RCIM_Cubic>]}
        GreenCurve {keys [<at 0.0938 0.4057>,<at 0.5194 0 RCIM_Cubic>]}
        RedCurve {keys [<at 0.0938 2>,<at 0.5194 1 RCIM_Cubic>]}
       }}))
    ScaleA=true
    ScaleRGB=true
   )
   &DynamicMaterialParameters(
    Index 0 Param 1=&RandomRangeFloat(Minimum=0, Maximum=1)
    Index 0 Param 2=&RandomRangeFloat(Minimum=0, Maximum=1)
    Index 0 Param 3=1
    Index 0 Param 4=0.7
    Index 1 Param 1=&FloatFromCurve(FloatCurve=DataInterface() {propertyValues {Curve {keys [
         {interpMode=RCIM_Cubic, tangentMode=RCTM_User, time=0, value=0.2}
         {arriveTangent=1.5012, interpMode=RCIM_Cubic, leaveTangent=1.5012, tangentMode=RCTM_User, time=1, value=0.5}
        ]}}})
    Index 1 Param 4=100
    Module.Index0Param0Write=true
    Module.Index0Param1Write=true
    Module.Index0Param2Write=true
    Module.Index0Param3Write=true
    Module.Index1Param0Write=true
    Module.Index1Param1Write=true
    Module.Index1Param2Write=true
    Module.Index1Param3Write=true
    Module.Index2Param0Write=true
    Module.Index2Param1Write=false
    Module.Index2Param2Write=false
    Module.Index2Param3Write=false
    Write Parameter Index 1=true
    Write Parameter Index 2=true
   )
   &ScaleColor(
    Scale Alpha=&Multiply_Float(
     A=1
     B=&FloatFromCurve(
      FloatCurve=DataInterface() {propertyValues {Curve {
         defaultValue=0
         keys [
          {interpMode=RCIM_Cubic, tangentMode=RCTM_User, time=0, value=0}
          {arriveTangent=1, interpMode=RCIM_Cubic, leaveTangent=1, time=0.3, value=0.175}
          {interpMode=RCIM_Cubic, tangentMode=RCTM_User, time=1, value=1}
         ]
        }}}
      CurveIndex=$Emitter.NormalizedLoopAge
     )
    )
    Scale RGB=&VectorFromFloat(Value=&FloatFromCurve(
      FloatCurve=DataInterface() {propertyValues {Curve {
         defaultValue=1
         keys [{interpMode=RCIM_Cubic, tangentMode=RCTM_User, time=0, value=0.9}, {interpMode=RCIM_Cubic, tangentMode=RCTM_User, time=1, value=1}]
        }}}
      CurveIndex=$Emitter.NormalizedLoopAge
     ))
    ScaleA=true
    ScaleRGB=true
   ) {module=ScaleColor001}
   &ScaleColor(
    Scale Alpha=&FloatFromCurve(FloatCurve=DataInterface() {propertyValues {Curve {
        defaultValue=1
        keys [{interpMode=RCIM_Cubic, tangentMode=RCTM_User, time=0, value=0}, {interpMode=RCIM_Cubic, tangentMode=RCTM_User, time=0.4, value=1}]
       }}})
    ScaleA=true
    ScaleRGB=true
   ) {module=ScaleColor002}
   &ScaleColor(Scale Alpha=0.3, Scale RGB=<8 0.8 0.8>, ScaleA=true, ScaleRGB=true) {module=ScaleColor003}
  ]
  rendererClasses [&NiagaraSpriteRendererProperties]
  renderers [{rendererClass=&NiagaraSpriteRendererProperties}]
 }
 Glow {
  bEnabled=false
  emitterUpdateScript [
   &EmitterState(Inactive Response=::NewEnumerator1, Life Cycle Mode=::NewEnumerator1, Loop Behavior=::NewEnumerator1, UseLoopDelay=true)
   &SpawnRate(SpawnRate=10, Use Spawn Probability=false)
  ]
  particleSpawnScript [
   &InitializeParticle(
    Color=<rgba 5.2249 1.115 0.5145 0.571>
    Lifetime=0.2
    Mesh Scale=&Multiply_Vector(
     A=&RandomRangeVector(Minimum=<1 1 1>, Maximum=<1 1.5 1.5>)
     B=&VectorFromFloat(Value=5)
    )
    Mesh Scale Mode=::NewEnumerator4
    Position Offset Coordinate Space=::Local
    UsePositionOffset=true
    Write Material Random=true
   )
   &InitialMeshOrientation(Mesh Orientation Mode=::NewEnumerator2, Rotation=<0 0.25 0>, Use Rotation Vector=true)
   /SetVariables(Particles.Particle_BG_Color:LinearColor=<rgba 0.1198 0 0.0764 0.545>) {isSetParameters=true}
  ]
  particleUpdateScript [
   &ScaleColor(Scale Alpha=&FN_FrontEndFloatModifier(Gameplay=1, Front End=0), ScaleA=true, ScaleRGB=true) {module=ScaleColor001}
   &ParticleState()
   &SpriteRotationRate(Rotation Rate=10)
   &DynamicMaterialParameters(
    Index 0 Param 1=1.2
    Index 0 Param 2=2
    Index 0 Param 3=1
    Index 0 Param 4=1
    Module.Index0Float0Write=true
    Module.Index0Float1Write=true
    Module.Index0Float2Write=true
    Module.Index0Float3Write=true
   )
   &ScaleMeshSize(Scale Factor=&Multiply_VectorByFloat(
     Vector=<0.6 0.5 0.5>
     Float=&FloatFromCurve(FloatCurve=DataInterface() {propertyValues {Curve {keys [{interpMode=RCIM_Cubic, tangentMode=RCTM_User, time=0, value=0.4}, {interpMode=RCIM_Cubic, tangentMode=RCTM_User, time=1, value=0.575}]}}})
    ))
   &SkeletalMeshLocation(
    Bone Sampling Mode=::NewEnumerator4
    MaskSpawnGroup=false
    Orientation Sampling=::NewEnumerator1
    Skeletal Mesh=DataInterface() {propertyValues {
      DefaultMesh=&SuperSpike_Gadget
      ExcludeBoneName=root
      OptionalComponentClass=null
      PreviewMesh=&SuperSpike_Gadget
      SkinningMode=SkinOnTheFly
      SoftSourceActor=null
      SourceMode=Default
      UvSetIndex=0
      WholeMeshLOD=-1
      bExcludeBone=false
      bReadDeformedGeometry=true
      bRequireCurrentFrameData=true
      ComponentTags []
      FilteredBones [JNT_Ball_A]
      FilteredSockets []
      MeshUserParameter {parameter {name=None, typeDefHandle {registeredTypeIndex=1}, varData []}}
      SamplingRegions []
     }}
   )
   /SetVariables(Particles.Position=&AddVectorToPosition(Position=$Particles.Position, Vector=<0 0 -21>)) {isSetParameters=true}
   &ScaleColor(Scale Alpha=0.8, ScaleA=true, ScaleRGB=true) {module=ScaleColor002}
   /SetVariables(
    Particles.BallGlowCameraOffset=20
    Particles.BallGlowColor:LinearColor=&Multiply_LinearColorByFloat(
     Linear Color=<rgba 0.9844 0.3761 0.1077 0.9>
     Float=&MakeFloatFromLinearColor(LinearColor=$Particles.Color, Channel:ENiagaraFloatFromLinearColorOptions=::NewEnumerator3)
    )
    Particles.BallGlowPosiiton:NiagaraPosition=&AddVectorToPosition(Position=$Particles.Position, Vector=<0 0 20>)
    Particles.BallGlowSize:Vector2f=<40 40>
   ) {isSetParameters=true}
   /SetVariables(
    Particles.EnergyPositionOffset:NiagaraPosition=&AddVectorToPosition(Position=$Particles.Position, Vector=<0 0 20>)
    Particles.MeshCameraOffset=-125
    Particles.ScaleOffset:Vector3f=&Add_Vector(A=$Particles.Scale, B=&VectorFromFloat(Value=0.2))
   ) {isSetParameters=true}
  ]
  rendererClasses [&NiagaraSpriteRendererProperties]
  renderers [
   {rendererClass=&NiagaraSpriteRendererProperties}
   {rendererClass=&NiagaraSpriteRendererProperties, rendererIndex=1}
  ]
 }
 Light {
  emitterUpdateScript [
   &EmitterState(Life Cycle Mode=::NewEnumerator1, Loop Behavior=::NewEnumerator1, Loop Duration=6, UseLoopDelay=false)
   &SpawnBurst_Instantaneous(Use Spawn Probability=false, UseLoopCountLimit=false)
  ]
  particleSpawnScript [&InitializeParticle(Color Mode=::NewEnumerator0, UsePositionOffset=true, Write Material Random=true)]
  particleUpdateScript [
   &ParticleState()
   &SystemLocation(MaskSpawnGroup=false, Offset=<0 0 6>, Offset Coordinate Space=::Simulation)
   &FN_Light_Attributes(
    Color=<rgba 8 2.3947 1.293 1>
    Create Color Attribute=true
    Create Exponent Attribute=true
    Create Position Attribute=true
    Create Radius Attribute=true
    Exponent=3
    Position=<5 0 0>
    Radius=100
    Scale Color=&FloatFromCurve(FloatCurve=DataInterface() {propertyValues {Curve {keys [<at 0 0>,<at 0.1 1>,<at 0.9 1>,<at 1 0>]}}})
   )
   &SkeletalMeshLocation(
    Bone Sampling Mode=::NewEnumerator4
    MaskSpawnGroup=false
    Orientation Sampling=::NewEnumerator1
    Skeletal Mesh=DataInterface() {propertyValues {
      DefaultMesh=&SuperSpike_Gadget
      ExcludeBoneName=root
      OptionalComponentClass=null
      PreviewMesh=&SuperSpike_Gadget
      SkinningMode=SkinOnTheFly
      SoftSourceActor=null
      SourceMode=Default
      UvSetIndex=0
      WholeMeshLOD=-1
      bExcludeBone=false
      bReadDeformedGeometry=true
      bRequireCurrentFrameData=true
      ComponentTags []
      FilteredBones [JNT_Ball_A]
      FilteredSockets []
      MeshUserParameter {parameter {name=None, typeDefHandle {registeredTypeIndex=1}, varData []}}
      SamplingRegions []
     }}
   )
  ]
  rendererClasses [&NiagaraLightRendererProperties]
  renderers [{rendererClass=&NiagaraLightRendererProperties, rendererIndex=0}]
 }
 Trail {
  emitterUpdateScript [
   &EmitterState(Life Cycle Mode=::NewEnumerator1, Loop Behavior=::NewEnumerator1, Loop Duration=0.8, UseLoopDelay=false)
   &SpawnRate(SpawnRate=200, Use Spawn Probability=false)
  ]
  particleSpawnScript [
   &SkeletalMeshLocation(
    Bone Sampling Mode=::NewEnumerator4
    MaskSpawnGroup=false
    Orientation Sampling=::NewEnumerator1
    Position Sampling=::NewEnumerator2
    Sampled Orientation Coordinate Space=::Local
    Sampled Position Coordinate Space=::Local
    Sampled Position Offset Coordinate Space:ENiagaraSkelMeshTransforms=::NewEnumerator2
    Sampled Velocity Coordinate Space:ENiagaraCoordinateSpace=::Local
    Skeletal Mesh=DataInterface() {propertyValues {
      DefaultMesh=&SuperSpike_Gadget
      ExcludeBoneName=root
      OptionalComponentClass=null
      PreviewMesh=&SuperSpike_Gadget
      SkinningMode=SkinOnTheFly
      SoftSourceActor=null
      SourceMode=Default
      UvSetIndex=0
      WholeMeshLOD=-1
      bExcludeBone=false
      bReadDeformedGeometry=true
      bRequireCurrentFrameData=true
      ComponentTags []
      FilteredBones [JNT_Ball_A]
      FilteredSockets []
      MeshUserParameter {parameter {name=None, typeDefHandle {registeredTypeIndex=1}, varData []}}
      SamplingRegions []
     }}
   )
   &InitializeParticle(
    Color=<rgba 0.1195 0 0.0762 0.05>
    Lifetime=0.15
    Ribbon Width=&FloatFromCurve(
     FloatCurve=DataInterface() {propertyValues {Curve {defaultValue=1, keys [{tangentMode=RCTM_User, time=0, value=0.275}, {tangentMode=RCTM_User, time=1, value=0.4}]}}}
     CurveIndex=$Emitter.NormalizedLoopAge
     Scale Curve=60
    )
    Ribbon Width Mode=::NewEnumerator1
    UsePositionOffset=false
    Write Material Random=true
   )
   /SetVariables(Particles.Position=&AddVectorToPosition(Position=$Particles.SkeletalMeshLocation.SampledPosition, Vector=<0 0 0>)) {isSetParameters=true}
  ]
  particleUpdateScript [
   &ParticleState()
   &SolveForcesAndVelocity(Clamp Velocity=false, Limit Acceleration=false)
   &ScaleRibbonWidth(Ribbon Width Scale=&FloatFromCurve(FloatCurve=DataInterface() {propertyValues {Curve {keys [<at 0 1>,<at 1 0>]}}}))
   &ScaleColor(
    Scale Alpha=&FloatFromCurve(FloatCurve=DataInterface() {propertyValues {Curve {defaultValue=1, keys []}}})
    ScaleA=true
    ScaleRGB=true
   )
   /SetVariables(
    Particles.TopTrailAge:NiagaraFloat=&Multiply_Float(A=$Particles.NormalizedAge, B=1)
    Particles.TopTrailColor:LinearColor=&MultiplyLinearColor(
     Linear Color=&MultiplyLinearColor(
      Linear Color=<rgba 1.901 0.3204 0 0.05>
      Scale Factor=&MakeVector4_FromLinearColor(LinearColor=&LerpLinearColors(
        StartColor=<rgba 1 1 1 1>
        EndColor=<rgba 1 1 1 1>
        LerpFactor=&FN_FrontEndFloatModifier(Gameplay=0, Front End=1)
       ))
     )
     Scale Factor=&MakeVector4(
      X=&FN_FrontEndFloatModifier(Gameplay=1, Front End=0.8)
      Y=&FN_FrontEndFloatModifier(Gameplay=1, Front End=1.5)
      Z=&FN_FrontEndFloatModifier(Gameplay=1, Front End=1)
      W=&FN_FrontEndFloatModifier(Gameplay=1, Front End=1)
     )
    )
   ) {isSetParameters=true}
  ]
  rendererClasses [&NiagaraRibbonRendererProperties]
  renderers [
   {rendererClass=&NiagaraRibbonRendererProperties}
   {rendererClass=&NiagaraRibbonRendererProperties, rendererIndex=1}
  ]
 }
 Volleyball_Outlines {
  emitterUpdateScript [
   &EmitterState(Life Cycle Mode=::NewEnumerator1, Loop Behavior=::NewEnumerator1, Loop Duration=1, UseLoopDelay=false)
   &SpawnBurst_Instantaneous(Use Spawn Probability=false, UseLoopCountLimit=false)
  ]
  particleSpawnScript [
   &InitializeParticle(Color=<rgba 1 0.3607 0 1>, Mesh Scale Mode=::NewEnumerator3, UsePositionOffset=false, Write Material Random=true)
   &InitialMeshOrientation(Mesh Orientation Mode=::NewEnumerator1, Use Rotation Vector=false)
  ]
  particleUpdateScript [
   &ParticleState()
   &SkeletalMeshLocation(
    Bone Sampling Mode=::NewEnumerator4
    MaskSpawnGroup=false
    Position Sampling=::NewEnumerator2
    Skeletal Mesh=DataInterface() {propertyValues {
      DefaultMesh=&SuperSpike_Gadget
      ExcludeBoneName=root
      OptionalComponentClass=null
      PreviewMesh=&SuperSpike_Gadget
      SkinningMode=SkinOnTheFly
      SoftSourceActor=null
      SourceMode=Default
      UvSetIndex=0
      WholeMeshLOD=-1
      bExcludeBone=false
      bReadDeformedGeometry=true
      bRequireCurrentFrameData=true
      ComponentTags []
      FilteredBones [JNT_Ball_A]
      FilteredSockets []
      MeshUserParameter {parameter {name=None, typeDefHandle {registeredTypeIndex=1}, varData []}}
      SamplingRegions []
     }}
   )
   /SetVariables(Particles.Position=&AddVectorToPosition(Position=$Particles.SkeletalMeshLocation.SampledPosition, Vector=<0 0 0>)) {isSetParameters=true}
   &ScaleColor(Scale Alpha=&FN_FrontEndFloatModifier(Gameplay=1, Front End=0.8), ScaleA=true, ScaleRGB=true)
  ]
  rendererClasses [&NiagaraMeshRendererProperties]
  renderers [{rendererClass=&NiagaraMeshRendererProperties, rendererIndex=0}]
 }
}
summary {
 systemName=NS_Emote_SuperSpike_Ball_Trail_Throw
 emitters [
  {emitterName=Trail, rendererClasses [&NiagaraRibbonRendererProperties]}
  {emitterName=BG, rendererClasses [&NiagaraSpriteRendererProperties]}
  {emitterName=Fire, rendererClasses [&NiagaraSpriteRendererProperties]}
  {emitterName=Light, rendererClasses [&NiagaraLightRendererProperties]}
  {bEnabled=false, emitterName=Glow, rendererClasses [&NiagaraSpriteRendererProperties]}
  {emitterName=Volleyball_Outlines, rendererClasses [&NiagaraMeshRendererProperties]}
 ]
 userVariables [
  User.bIsFrontEnd {type=NiagaraBool}
  User.bIsFrontEndPreview {type=NiagaraBool}
  User.DustSpawnCurve {type=NiagaraFloat, defaultValue=1, description=""}
  User.Particle_BG_Color {type=LinearColor, description="", defaultValue {struct=LinearColor, value=<rgba 1 1 1 1>}}
  User.SkeletalMesh {
   type=NiagaraDataInterfaceSkeletalMesh
   defaultValue {
    struct=NiagaraExt_VariableValue_DataInterface
    value {dataInterface=&self:NiagaraDataInterfaceSkeletalMesh_0}
   }
  }
 ]
}
